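; Weapon Blueprint for the FirePetal Quirky SMG that manages barrel spin audio and unequip cleanup.
; For locally viewed pawns, firing spawns an attached sound once, then reuses it via Retrigger; fire stop sends Stop.
; An equip-time delegate feeds aim-down-sights state into the sound's IsADS parameter.
; Unequip fades the audio, and simulated proxies stop the weapon mesh's active montage unless the weapon shows when holstered.
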(event EventPlayWeaponFireCustomAudio
  (params (bPersistentFire :type Boolean) (bSecondaryFire :type Boolean))
  (bind _asfort_player_pawn_athena (Utilities|Casting|CastToFortPlayerPawnAthena (Game|GetInstigator)))
  (if (Pawn|IsLocallyViewed _asfort_player_pawn_athena)
    (bind _barrel_spin_audio_component (Variables|Default|GetBarrel%20Spin%20Audio%20Component)
      (:then
        (Audio|Parameter|ExecuteTriggerParameter _barrel_spin_audio_component "Retrigger"))
      (:else
        (bind _returnvalue (Audio|SpawnSoundAttached (Variables|Default|GetBarrel%20Spin%20Sound) (Variables|Transformation|GetRootComponent)))
        (Variables|Default|SetBarrel%20Spin%20Audio%20Component _returnvalue)))))

(event EventStopWeaponFireCustomAudio
  (bind _barrel_spin_audio_component (Variables|Default|GetBarrel%20Spin%20Audio%20Component)
    (:then
      (Audio|Parameter|ExecuteTriggerParameter _barrel_spin_audio_component "Stop"))
    (:else)))

(event FortWeapon|EventOnUnEquippedAndDetached
  (bind _weaponmesh (Variables|Visual|GetWeaponMesh))
  (bind _returnvalue (Components|SkeletalMesh|GetAnimInstance _weaponmesh))
  (bind _returnvalue_2 (Animation|Montage|GetCurrentActiveMontage _returnvalue))
  (Utilities|FlowControl|Sequence
    (:then_0
      (bind _barrel_spin_audio_component (Variables|Default|GetBarrel%20Spin%20Audio%20Component)
        (:then
          (Audio|Components|Audio|FadeOut _barrel_spin_audio_component 0.2 0.0 "Linear" :!class /Script/Engine.AudioComponent))
        (:else)))
    (:then_1
      (switch Utilities|FlowControl|Switch|SwitchonENetRole (Networking|GetLocalRole)
        (:ROLE_SimulatedProxy
          (Utilities|IsValid _returnvalue_2
            (:Is%20Valid
              (if (not (Variables|FortWeapon|GetbShowWhenHolstered))
                (Animation|Montage|MontageStop _returnvalue (Montage|GetDefaultBlendOutTime _returnvalue_2) _returnvalue_2)))))))))

(event FortWeapon|EventOnEquippedAndAttached
  (bind _asfort_player_pawn_athena (Utilities|Casting|CastToFortPlayerPawnAthena (Game|GetInstigator)))
  (if (Pawn|IsLocallyViewed _asfort_player_pawn_athena)
    (Default|BindEventtoOnAimDownSightsChanged (delegate Custom|onAimDownSightsChanged_Event))))

(event Custom|onAimDownSightsChanged_Event
  (params (AimDownSights :type Boolean))
  (bind _barrel_spin_audio_component (Variables|Default|GetBarrel%20Spin%20Audio%20Component)
    (:then
      (Audio|Parameter|SetBooleanParameter _barrel_spin_audio_component "IsADS" AimDownSights :!class /Script/AudioExtensions.AudioParameterControllerInterface))
    (:else)))

(fn ConstructionScript
  (|CallParentFunction))
