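# Sound cue for the drop sound of the FirePetal Quirky SMG weapon.
# A Random node picks one of three equally weighted drop wave variations, then feeds it through a Modulator at the root for variation.
# It uses the shared weapon-drop attenuation, the loot-collect sound class and a stop-oldest weapon-drop concurrency set.
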SoundCue /FirePetalWeaponGameplay/Gameplay/FirePetal_SMG_Quirky/Sound/FirePetal_SMG_Quirky_Drop_Cue
root=Modulator_1
nodes [
 {id=Modulator_1, children [Random_2]}
 {id=Random_2, children [WavePlayer_6,WavePlayer_7,WavePlayer_8], properties {weights [1,1,1]}}
 {
  id=WavePlayer_6
  properties {soundWaveAssetPtr=/FirePetalWeaponGameplay/Gameplay/FirePetal_SMG_Quirky/Sound/SoundWaves/FirePetal_SMG_Quirky_Drop_01}
 }
 {
  id=WavePlayer_7
  properties {soundWaveAssetPtr=/FirePetalWeaponGameplay/Gameplay/FirePetal_SMG_Quirky/Sound/SoundWaves/FirePetal_SMG_Quirky_Drop_02}
 }
 {
  id=WavePlayer_8
  properties {soundWaveAssetPtr=/FirePetalWeaponGameplay/Gameplay/FirePetal_SMG_Quirky/Sound/SoundWaves/FirePetal_SMG_Quirky_Drop_03}
 }
]
properties {
 attenuationSettings=&Fort_Weapon_Drop_Att
 maxDistance=3000
 soundClassObject=&Fort_Loot_Collect
 volumeMultiplier=1
 concurrencyOverrides {resolutionRule=PreventNew}
 concurrencySet [&Weapon_Drop_StopOldest]
}
