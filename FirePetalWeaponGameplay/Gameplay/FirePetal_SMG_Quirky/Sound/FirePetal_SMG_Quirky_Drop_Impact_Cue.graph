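# Sound cue for the impact sound of the FirePetal Quirky SMG when dropped.
# A Modulator at the root feeds from a short Delay, which feeds from an evenly weighted Random choice among three impact waves.
# It uses the shared weapon-drop attenuation and loot-collect sound class, with a concurrency override that blocks new instances once a few are playing.
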SoundCue /FirePetalWeaponGameplay/Gameplay/FirePetal_SMG_Quirky/Sound/FirePetal_SMG_Quirky_Drop_Impact_Cue
root=Modulator_0
nodes [
 {id=Delay_0, children [Random_0], properties {delayMin=0.1}}
 {id=Modulator_0, children [Delay_0]}
 {id=Random_0, children [WavePlayer_3,WavePlayer_4,WavePlayer_5], properties {weights [1,1,1]}}
 {
  id=WavePlayer_3
  properties {soundWaveAssetPtr=/FirePetalWeaponGameplay/Gameplay/FirePetal_SMG_Quirky/Sound/SoundWaves/FirePetal_SMG_Quirky_Drop_Impact_01}
 }
 {
  id=WavePlayer_4
  properties {soundWaveAssetPtr=/FirePetalWeaponGameplay/Gameplay/FirePetal_SMG_Quirky/Sound/SoundWaves/FirePetal_SMG_Quirky_Drop_Impact_02}
 }
 {
  id=WavePlayer_5
  properties {soundWaveAssetPtr=/FirePetalWeaponGameplay/Gameplay/FirePetal_SMG_Quirky/Sound/SoundWaves/FirePetal_SMG_Quirky_Drop_Impact_03}
 }
]
properties {
 attenuationSettings=&Fort_Weapon_Drop_Att
 bOverrideConcurrency=true
 maxDistance=3000
 soundClassObject=&Fort_Loot_Collect
 volumeMultiplier=1.2
 attenuationOverrides {falloffDistance=1000, attenuationShapeExtents {x=0}}
 concurrencyOverrides {maxCount=3, resolutionRule=PreventNew}
}
