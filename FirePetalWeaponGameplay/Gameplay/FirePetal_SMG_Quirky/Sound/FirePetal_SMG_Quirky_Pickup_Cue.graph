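# Pickup sound cue for the FirePetal Quirky SMG weapon, classed as loot collection audio.
# A random node picks one of three equally weighted pickup waves, then a modulator adds slight pitch variation.
# Weapon pickup attenuation shapes falloff, and the cue-level multipliers raise pitch and lower volume.
# It references a shared pickup concurrency set, alongside override values for a single instance that prevents new plays.
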SoundCue /FirePetalWeaponGameplay/Gameplay/FirePetal_SMG_Quirky/Sound/FirePetal_SMG_Quirky_Pickup_Cue
root=Modulator_0
nodes [
 {id=Modulator_0, children [Random_1], properties {pitchMax=1.02, pitchMin=0.99, volumeMin=1}}
 {id=Random_1, children [WavePlayer_3,WavePlayer_4,WavePlayer_5], properties {weights [1,1,1]}}
 {
  id=WavePlayer_3
  properties {soundWaveAssetPtr=/FirePetalWeaponGameplay/Gameplay/FirePetal_SMG_Quirky/Sound/SoundWaves/FirePetal_SMG_Quirky_Pickup_03}
 }
 {
  id=WavePlayer_4
  properties {soundWaveAssetPtr=/FirePetalWeaponGameplay/Gameplay/FirePetal_SMG_Quirky/Sound/SoundWaves/FirePetal_SMG_Quirky_Pickup_02}
 }
 {
  id=WavePlayer_5
  properties {soundWaveAssetPtr=/FirePetalWeaponGameplay/Gameplay/FirePetal_SMG_Quirky/Sound/SoundWaves/FirePetal_SMG_Quirky_Pickup_01}
 }
]
properties {
 attenuationSettings=&Fort_Weapon_Pickup_Att
 maxDistance=3500
 pitchMultiplier=1.1
 soundClassObject=&Fort_Loot_Collect
 volumeMultiplier=0.6
 concurrencyOverrides {maxCount=1, resolutionRule=PreventNew}
 concurrencySet [&PickupMaximum_Oldest]
}
