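# Stereo one-shot MetaSound for the FirePetal Quirky SMG barrel spin, playing a looping wave until it winds down.
# FireRate sets the pitch offset and ramp times through delay pitch shifters and interpolated gain, with an ADS volume boost.
# Stop, or FireRate falling to its floor, fades out and drops pitch.
# OnFinished follows the fade; Retrigger cancels it and, accumulated with Stop, restarts playback.
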MetaSoundSource /FirePetalWeaponGameplay/Gameplay/FirePetal_SMG_Quirky/Sound/MSS_FirePetal_SMG_Quirky_BarrelSpin
inputs [
 ADSVolumeMultiply {type=Float}
 FireRate {type=Float}
 IsADS {type=Bool}
 MasterVolume {type=Float}
 Retrigger {type=Trigger}
 Stop {type=Trigger}
 UE.Source.OnPlay {type=Trigger}
 Wave Asset {type=WaveAsset}
]
interfaces [UE.OutputFormat.Stereo:1.0, UE.Source.OneShot:1.0, UE.Source:1.1]
outputs [
 "UE.OutputFormat.Stereo.Audio:0" {type=Audio}
 "UE.OutputFormat.Stereo.Audio:1" {type=Audio}
 UE.Source.OneShot.OnFinished {type=Trigger}
]
pages [Default {
  input_defaults {
   ADSVolumeMultiply=1.4
   FireRate=7.25
   IsADS=null
   MasterVolume=0.6
   Retrigger=null
   Stop=null
   UE.Source.OnPlay=null
   Wave Asset=/FirePetalWeaponGameplay/Gameplay/FirePetal_SMG_Quirky/Sound/SoundWaves/FirePetal_SMG_BarrelSpin_Loop
  }
  nodes [
   InitVariable.Trigger()
   n1 = MapRange.MapRange:Float(In=^FireRate) {also {defaults {In Range A=7.25, In Range B=14, Out Range A=0.5, Out Range B=0.75}}}
   n2 = UE.Subtract:Float(AdditionalOperands=MapRange.MapRange:Float(In=^FireRate) {also {defaults {In Range A=7.25, In Range B=14, Out Range A=0.5, Out Range B=0.25}}}.Out Value) {also {defaults {PrimaryOperand=1}}}
   TriggerDelay_1 = UE.Trigger Delay(In=^Stop) {also {defaults {Delay Time=0.03}}}
   TriggerAccumulatorTriggerAccumulate2_0 = "TriggerAccumulator.Trigger Accumulate (2)"(In 0=^Retrigger, In 1=@TriggerDelay_1.Out) {also {defaults {Auto Reset="<MetaSoundBuilderResult.SUCCEEDED: 0>"}}}
   TriggerAnyTriggerAny2_0 = "TriggerAny.Trigger Any (2)"(In 0=^UE.Source.OnPlay, In 1=@TriggerAccumulatorTriggerAccumulate2_0.Out)
   n3 = VariableAccessor.Trigger(Variable=VariableMutator.Trigger(Variable=InitVariable.Trigger().Variable, Value=@TriggerAnyTriggerAny2_0.Out).Variable)
   n4 = VariableAccessor.Trigger(Variable=@n3.Variable)
   TriggerAnyTriggerAny2_1 = "TriggerAny.Trigger Any (2)"(In 0=@TriggerDelay_1.Out, In 1=UE.TriggerOnThreshold:Float(In=^FireRate) {also {defaults {Threshold=7.26, Type=1}}}.Out)
   n5 = VariableAccessor.Trigger(Variable=VariableMutator.Trigger(Variable=InitVariable.Trigger().Variable, Value=@TriggerAnyTriggerAny2_1.Out).Variable)
   n6 = VariableAccessor.Trigger(Variable=@n5.Variable)
   n7 = Value.Value:Float(Init Value=@n1.Out Value, Target Value=@n2.Out, Reset=@n4.Value, Set=@n6.Value)
   n8 = UE.Add:Float(
    AdditionalOperands=UE.InterpTo:Audio(
     Interp Time=UE.ConversionFloatToTime(In=@n7.Output Value).Out
     Target=Value.Value:Float(Reset=@n4.Value, Set=@n6.Value) {also {defaults {Init Value=0, Target Value=-3}}}.Output Value
    ).Value
    PrimaryOperand=MapRange.MapRange:Float(In=^FireRate) {also {defaults {In Range A=7.25, In Range B=14, Out Range A=-3, Out Range B=2}}}.Out Value
   )
   WavePlayer_0 = UE.Wave Player:Stereo(Wave Asset=^Wave Asset, Play=VariableAccessor.Trigger(Variable=@n4.Variable).Value) {also {defaults {Loop="<MetaSoundBuilderResult.SUCCEEDED: 0>"}}}
   DelayPitchShift_0 = UE.Delay Pitch Shift:Audio(Pitch Shift=@n8.Out, In=@WavePlayer_0.Out Left)
   n9 = Convert.Float:Bool(Bool=^IsADS)
   n10 = UE.Multiply:Float(PrimaryOperand=^ADSVolumeMultiply, AdditionalOperands=@n9.Float)
   n11 = UE.Subtract:Float(AdditionalOperands=@n9.Float) {also {defaults {PrimaryOperand=1}}}
   n12 = UE.ConversionFloatToTime(In=Value.Value:Float(Init Value=@n1.Out Value, Target Value=@n2.Out, Reset=@n5.Value, Set=@n3.Value).Output Value)
   n13 = Value.Value:Float(Reset=@n5.Value, Set=@n3.Value) {also {defaults {Init Value=0, Target Value=1}}}
   n14 = UE.Multiply:Float(
    AdditionalOperands=^MasterVolume
    PrimaryOperand=UE.Multiply:Float(
     AdditionalOperands=UE.Add:Float(PrimaryOperand=@n10.Out, AdditionalOperands=@n11.Out).Out
     PrimaryOperand=UE.InterpTo:Audio(Interp Time=@n12.Out, Target=@n13.Output Value).Value
    ).Out
   )
   Multiply_0 = UE.Multiply:Audio by Float(PrimaryOperand=@DelayPitchShift_0.Out, AdditionalOperands=@n14.Out)
   DelayPitchShift_1 = UE.Delay Pitch Shift:Audio(Pitch Shift=@n8.Out, In=@WavePlayer_0.Out Right)
   Multiply_1 = UE.Multiply:Audio by Float(PrimaryOperand=@DelayPitchShift_1.Out, AdditionalOperands=@n14.Out)
   TriggerDelay_0 = UE.Trigger Delay(Reset=^Retrigger, In=UE.Pipe(Delay Time=@n12.Out, In=@n13.On Reset, Reset=@n13.On Set).Out)
  ]
  links [
   {dst="@UE.OutputFormat.Stereo.Audio:0", src=Multiply_0.Out}
   {dst="@UE.OutputFormat.Stereo.Audio:1", src=Multiply_1.Out}
   {dst="@UE.Source.OneShot.OnFinished", src=TriggerDelay_0.Out}
  ]
 }]
properties {
 assetImportData=&self:AssetImportData
 attenuationSettings=/Game/Sounds/Attenuation/Fort_Weapon_Rifle_Lng_Att
 bOverrideConcurrency=true
 maxDistance=7500
 numChannels=2
 outputFormat=Stereo
 sampleRate=48000
 soundClassObject=&Fort_WeaponRanged
 concurrencyOverrides {maxCount=3}
}
